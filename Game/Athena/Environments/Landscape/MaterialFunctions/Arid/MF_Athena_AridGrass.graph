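# Landscape material function that layers tinted arid grass color over an incoming material for Athena environments.
# Grass base color is multiplied by the Arid_Grass_Tint parameter from MPC_Landscape.
# HeightLerp turns the Mask input and the grass Opacity, used as height, into the blend alpha, taken without contrast.
# Material_Standard changes only BaseColor, while Material_Low outputs the same blend over a separate base color input.
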MaterialFunction /Game/Athena/Environments/Landscape/MaterialFunctions/Arid/MF_Athena_AridGrass
graph {nodes [
  n1 = GetMaterialAttributes(None=FunctionInput() {inputName=Material_Standard, inputType=FunctionInput_MaterialAttributes}) {attributeGetTypes [#1]}
  n2 = GetMaterialAttributes(None=FunctionInput() {inputName=Grass, inputType=FunctionInput_MaterialAttributes}) {attributeGetTypes [#1, #2]}
  n3 = ComponentMask(None=CollectionParameter() {collection=/Game/Athena/Environments/Landscape/MPC/MPC_Landscape, parameterName=Arid_Grass_Tint}) {b=true, g=true, r=true}
  n4 = Multiply(A=@n2.BaseColor, B=@n3)
  n5 = MaterialFunctionCall(Transition Phase=FunctionInput() {inputName=Mask, inputType=FunctionInput_Scalar, sortPriority=2}, Height Texture=@n2.Opacity) {materialFunction=&HeightLerp}
  FunctionOutput(None=SetMaterialAttributes!(MaterialAttributes=@n1.MaterialAttributes, BaseColor=LinearInterpolate(A=@n1.BaseColor, B=@n4, Alpha=@n5.Lerp Alpha No Contrast)) {attributeSetTypes [#1]}) {outputName=Material_Standard}
  FunctionOutput(None=LinearInterpolate(A=FunctionInput() {inputName=Material_BaseColor, sortPriority=1}, B=Multiply(A=@n2.BaseColor, B=@n3), Alpha=@n5.Lerp Alpha No Contrast)) {outputName=Material_Low, sortPriority=1}
 ]}
